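# Frontend physics asset for the CrispRover male medium player body, set up for simulating its coat tails.
# Capsules cover the pelvis, thighs and calves, while box bodies cover left and right back and side coat chains, three bones each.
# Each coat chain hangs from the pelvis with angular limits; back roots lock twist and side roots lock one swing axis.
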PhysicsAsset /BRCosmetics/Characters/Player/Male/Medium/Bodies/M_MED_CrispRover/Meshes/M_MED_CrispRover_Physics_Frontend
bodies {
 calf_l [{shapeType=Capsule, length=38, radius=9.9, center=<-24.0188 -3.1203 1.3234>, rotation=<rot 89.5951 80.8269 80.5139>}]
 calf_r [{shapeType=Capsule, length=38, radius=8, center=<21.2464 1.819 -0.4419>, rotation=<rot 89.5951 80.8269 80.5139>}]
 dyn_L_backCoat_1 [{shapeType=Box, extentX=1.138, extentY=25, extentZ=20, shapeName=dyn_L_backCoat_1_box, center=<5 0 0>, rotation=<rot 90 0 0>}]
 dyn_L_backCoat_2 [{shapeType=Box, extentX=2.5391, extentY=25, extentZ=15, shapeName=dyn_L_backCoat_2_box, center=<7 0 0>, rotation=<rot 90 0 0>}]
 dyn_L_backCoat_3 [{shapeType=Box, extentX=2.5391, extentY=25, extentZ=20, shapeName=dyn_L_backCoat_3_box, center=<6 0 0>, rotation=<rot 90 0 0>}]
 dyn_L_sideCoat_1 [{shapeType=Box, extentX=2.9883, extentY=30, extentZ=15, shapeName=dyn_L_sideCoat_1_box, center=<7 0 2>, rotation=<rot 85 0 0>}]
 dyn_L_sideCoat_2 [{shapeType=Box, extentX=4.5093, extentY=30, extentZ=15, shapeName=dyn_L_sideCoat_2_box, center=<7 0 1>, rotation=<rot 85 0 0>}]
 dyn_L_sideCoat_3 [{shapeType=Box, extentX=3.5788, extentY=30, extentZ=15, shapeName=dyn_L_sideCoat_3_box, center=<7 0 1>, rotation=<rot 85 0 0>}]
 dyn_R_backCoat_1 [{shapeType=Box, extentX=1.138, extentY=25, extentZ=20, shapeName=dyn_L_backCoat_1_box, center=<-5 0 0>, rotation=<rot 90 0 0>}]
 dyn_R_backCoat_2 [{shapeType=Box, extentX=2.5391, extentY=25, extentZ=15, shapeName=dyn_L_backCoat_2_box, center=<-7 0 0>, rotation=<rot 90 0 0>}]
 dyn_R_backCoat_3 [{shapeType=Box, extentX=2.5391, extentY=25, extentZ=20, shapeName=dyn_L_backCoat_3_box, center=<-6 0 0>, rotation=<rot 90 0 0>}]
 dyn_R_sideCoat_1 [{shapeType=Box, extentX=4.4215, extentY=30, extentZ=15, shapeName=dyn_L_sideCoat_1_box, center=<-7 0 -2>, rotation=<rot 85 0 0>}]
 dyn_R_sideCoat_2 [{shapeType=Box, extentX=6.3659, extentY=30, extentZ=15, shapeName=dyn_L_sideCoat_2_box, center=<-7 0 -1>, rotation=<rot 85 0 0>}]
 dyn_R_sideCoat_3 [{shapeType=Box, extentX=4.7923, extentY=30, extentZ=15, shapeName=dyn_L_sideCoat_3_box, center=<-7 0 -1>, rotation=<rot 85 0 0>}]
 pelvis [{shapeType=Capsule, length=1, radius=1, center=<0 -0.6425 0>}]
 thigh_l [{shapeType=Capsule, length=39.1255, radius=9, center=<-16.4823 -0.3283 2.3773>, rotation=<rot 88.4329 -0.2777 -4.442>}]
 thigh_r [{shapeType=Capsule, length=43.5146, radius=9, center=<16.5747 0.5758 -2.4153>, rotation=<rot 88.3549 10 10>}]
}
constraints [
 {bone1Name=thigh_l, bone2Name=pelvis, swing1LimitDegrees=45, swing2LimitDegrees=45, twistLimitDegrees=45}
 {bone1Name=calf_l, bone2Name=thigh_l, swing1LimitDegrees=45, swing2LimitDegrees=45, twistLimitDegrees=45}
 {bone1Name=thigh_r, bone2Name=pelvis, swing1LimitDegrees=45, swing2LimitDegrees=45, twistLimitDegrees=45}
 {bone1Name=calf_r, bone2Name=thigh_r, swing1LimitDegrees=45, swing2LimitDegrees=45, twistLimitDegrees=45}
 {bone1Name=dyn_L_backCoat_1, bone2Name=pelvis, swing1LimitDegrees=3, swing2LimitDegrees=30, twistLimitDegrees=45, twistMotion=Locked}
 {bone1Name=dyn_L_backCoat_2, bone2Name=dyn_L_backCoat_1, swing1LimitDegrees=3, swing2LimitDegrees=45, twistLimitDegrees=5}
 {bone1Name=dyn_L_backCoat_3, bone2Name=dyn_L_backCoat_2, swing1LimitDegrees=3, swing2LimitDegrees=45, twistLimitDegrees=5}
 {bone1Name=dyn_L_sideCoat_2, bone2Name=dyn_L_sideCoat_1, swing1LimitDegrees=20, swing2LimitDegrees=45, twistLimitDegrees=20}
 {bone1Name=dyn_L_sideCoat_1, bone2Name=pelvis, swing1LimitDegrees=20, swing1Motion=Locked, swing2LimitDegrees=30, twistLimitDegrees=20}
 {bone1Name=dyn_L_sideCoat_3, bone2Name=dyn_L_sideCoat_2, swing1LimitDegrees=20, swing2LimitDegrees=45, twistLimitDegrees=20}
 {bone1Name=dyn_R_backCoat_1, bone2Name=pelvis, swing1LimitDegrees=3, swing2LimitDegrees=30, twistLimitDegrees=45, twistMotion=Locked}
 {bone1Name=dyn_R_backCoat_2, bone2Name=dyn_R_backCoat_1, swing1LimitDegrees=3, swing2LimitDegrees=45, twistLimitDegrees=5}
 {bone1Name=dyn_R_backCoat_3, bone2Name=dyn_R_backCoat_2, swing1LimitDegrees=3, swing2LimitDegrees=45, twistLimitDegrees=5}
 {bone1Name=dyn_R_sideCoat_1, bone2Name=pelvis, swing1LimitDegrees=20, swing1Motion=Locked, swing2LimitDegrees=30, twistLimitDegrees=20}
 {bone1Name=dyn_R_sideCoat_2, bone2Name=dyn_R_sideCoat_1, swing1LimitDegrees=20, swing2LimitDegrees=45, twistLimitDegrees=20}
 {bone1Name=dyn_R_sideCoat_3, bone2Name=dyn_R_sideCoat_2, swing1LimitDegrees=20, swing2LimitDegrees=45, twistLimitDegrees=20}
]
