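# M_MED_CrispRover cosmetic body AnimBlueprint that adds secondary motion to coat, collars, belts, flaps, bracelets and boot straps.
# The parent mesh's copied pose feeds a Control Rig, corrective bones and two RigidBody physics assets weighted by custom-idle state.
# Cached poses split front-coat handling between custom idle and gameplay correctives, then AnimDynamics simulates accessories.
# EventGraph derives corrective alphas from bone deltas against reference pose, such as thighs, adjusted for crouching.
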AnimBlueprint /BRCosmetics/Characters/Player/Male/Medium/Bodies/M_MED_CrispRover/Meshes/M_MED_CrispRover_AnimBP
target_skeleton=&/M_MED_CrispRover_Skeleton
graphs [
 AnimGraph {nodes [
   n1 = "Transform (Modify) Bone - Bone: dyn_L_backCoat_master"!(ComponentPose, bAlphaBoolEnabled:bool=True, Translation="-18,0,16", Rotation="60,0,0", Scale="1,1,1") -> (Pose) {
    alphaInputType=Bool
    rotationMode=BMM_Additive
    rotationSpace=BCS_BoneSpace
    translationMode=BMM_Additive
    translationSpace=BCS_BoneSpace
    alphaBoolBlend {blendInTime=0.1, blendOption=Linear, blendOutTime=0.2}
    boneToModify {boneName=dyn_L_backCoat_master}
    rotation {pitch=60}
    translation {x=-18, z=16}
   }
   "Transform (Modify) Bone - Bone: dyn_R_backCoat_master"!(ComponentPose=@n1, bAlphaBoolEnabled:bool=True, Translation="18,0,-16", Rotation="60,0,0", Scale="1,1,1") -> (Pose) {
    alphaInputType=Bool
    rotationMode=BMM_Additive
    rotationSpace=BCS_BoneSpace
    translationMode=BMM_Additive
    translationSpace=BCS_BoneSpace
    alphaBoolBlend {blendInTime=0.1, blendOption=Linear, blendOutTime=0.2}
    boneToModify {boneName=dyn_R_backCoat_master}
    rotation {pitch=60}
    translation {x=18, z=-16}
   }
   n2 = "Constraint - dyn_L_frontCoat_1 (R)"(ComponentPose=Local To Component(LocalPose=Use cached pose 'preFixCache'()), Alpha=1, ConstraintWeights_0=1) {
    alphaBoolBlend {blendOption=Linear}
    boneToModify {boneName=dyn_L_frontCoat_1}
    constraintSetup [{transformType=Rotation, targetBone {boneName=thigh_l}}]
    constraintWeights [1]
   }
   n3 = "Constraint - dyn_L_frontCoat_2 (R)"(ComponentPose=@n2, Alpha=1, ConstraintWeights_0=1) {
    alphaBoolBlend {blendOption=Linear}
    boneToModify {boneName=dyn_L_frontCoat_2}
    constraintSetup [{transformType=Rotation, targetBone {boneName=thigh_l}}]
    constraintWeights [1]
   }
   n4 = "Constraint - dyn_L_frontCoat_3 (R)"(ComponentPose=@n3, Alpha=1, ConstraintWeights_0=1) {
    alphaBoolBlend {blendOption=Linear}
    boneToModify {boneName=dyn_L_frontCoat_3}
    constraintSetup [{transformType=Rotation, targetBone {boneName=thigh_l}}]
    constraintWeights [1]
   }
   n5 = Anim Dynamics - Bone: dyn_L_frontCoat_1(ComponentPose=@n4, Alpha=1, ExternalForce="0,0,0") {
    angularDampingOverride=1
    angularSpringConstant=50
    bOverrideAngularDamping=true
    simulationSpace=RootRelative
    alphaBoolBlend {blendOption=Linear}
    boundBone {boneName=dyn_L_frontCoat_1}
    physicsBodyDefinitions [{
      boundBone {boneName=dyn_L_frontCoat_1}
      boxExtents=<10 10 10>
      constraintSetup {angularLimitsMax=<45 2 45>, angularLimitsMin=<-45 -40 -45>, angularTarget=<1 0 0>}
      localJointOffset=<1 0 0>
     }]
    relativeSpaceBone {boneName=root}
   }
   n6 = Anim Dynamics - Bone: dyn_L_frontCoat_2(ComponentPose=@n5, Alpha=1, ExternalForce="0,0,0") {
    angularDampingOverride=1
    angularSpringConstant=50
    bOverrideAngularDamping=true
    simulationSpace=RootRelative
    alphaBoolBlend {blendOption=Linear}
    boundBone {boneName=dyn_L_frontCoat_2}
    physicsBodyDefinitions [{
      boundBone {boneName=dyn_L_frontCoat_2}
      boxExtents=<10 10 10>
      constraintSetup {angularLimitsMax=<45 2 45>, angularLimitsMin=<-45 -40 -45>, angularTarget=<1 0 0>}
      localJointOffset=<1 0 0>
     }]
    relativeSpaceBone {boneName=root}
   }
   n7 = Anim Dynamics - Bone: dyn_L_frontCoat_3(ComponentPose=@n6, Alpha=1, ExternalForce="0,0,0") {
    angularDampingOverride=1
    angularSpringConstant=50
    bOverrideAngularDamping=true
    simulationSpace=RootRelative
    alphaBoolBlend {blendOption=Linear}
    boundBone {boneName=dyn_L_frontCoat_3}
    physicsBodyDefinitions [{
      boundBone {boneName=dyn_L_frontCoat_3}
      boxExtents=<10 10 10>
      constraintSetup {angularLimitsMax=<45 2 45>, angularLimitsMin=<-45 -40 -45>, angularTarget=<1 0 0>}
      localJointOffset=<1 0 0>
     }]
    relativeSpaceBone {boneName=root}
   }
   n8 = "Transform (Modify) Bone - Bone: dyn_L_frontCoat_master"(ComponentPose=@n7, Alpha=1, Translation="0,0,0", Rotation="-1,-5,21", Scale="1,1,1") {
    rotationMode=BMM_Additive
    rotationSpace=BCS_BoneSpace
    alphaBoolBlend {blendOption=Linear}
    boneToModify {boneName=dyn_L_frontCoat_master}
    rotation=<rot -1 -5 21>
   }
   n9 = "Constraint - dyn_R_frontCoat_1 (R)"(ComponentPose=@n8, Alpha=1, ConstraintWeights_0=1) {
    alphaBoolBlend {blendOption=Linear}
    boneToModify {boneName=dyn_R_frontCoat_1}
    constraintSetup [{transformType=Rotation, targetBone {boneName=thigh_r}}]
    constraintWeights [1]
   }
   n10 = "Constraint - dyn_R_frontCoat_2 (R)"(ComponentPose=@n9, Alpha=1, ConstraintWeights_0=1) {
    alphaBoolBlend {blendOption=Linear}
    boneToModify {boneName=dyn_R_frontCoat_2}
    constraintSetup [{transformType=Rotation, targetBone {boneName=thigh_r}}]
    constraintWeights [1]
   }
   n11 = "Constraint - dyn_R_frontCoat_3 (R)"(ComponentPose=@n10, Alpha=1, ConstraintWeights_0=1) {
    alphaBoolBlend {blendOption=Linear}
    boneToModify {boneName=dyn_R_frontCoat_3}
    constraintSetup [{transformType=Rotation, targetBone {boneName=thigh_r}}]
    constraintWeights [1]
   }
   n12 = Anim Dynamics - Bone: dyn_R_frontCoat_1(ComponentPose=@n11, Alpha=1, ExternalForce="0,0,0") {
    angularDampingOverride=1
    angularSpringConstant=50
    bOverrideAngularDamping=true
    simulationSpace=RootRelative
    alphaBoolBlend {blendOption=Linear}
    boundBone {boneName=dyn_R_frontCoat_1}
    physicsBodyDefinitions [{
      boundBone {boneName=dyn_R_frontCoat_1}
      boxExtents=<10 10 10>
      constraintSetup {angularLimitsMax=<45 2 45>, angularLimitsMin=<-45 -40 -45>, angularTarget=<1 0 0>}
      localJointOffset=<-1 0 0>
     }]
    relativeSpaceBone {boneName=root}
   }
   n13 = Anim Dynamics - Bone: dyn_R_frontCoat_2(ComponentPose=@n12, Alpha=1, ExternalForce="0,0,0") {
    angularDampingOverride=1
    angularSpringConstant=50
    bOverrideAngularDamping=true
    simulationSpace=RootRelative
    alphaBoolBlend {blendOption=Linear}
    boundBone {boneName=dyn_R_frontCoat_2}
    physicsBodyDefinitions [{
      boundBone {boneName=dyn_R_frontCoat_2}
      boxExtents=<10 10 10>
      constraintSetup {angularLimitsMax=<45 2 45>, angularLimitsMin=<-45 -40 -45>, angularTarget=<1 0 0>}
      localJointOffset=<-1 0 0>
     }]
    relativeSpaceBone {boneName=root}
   }
   n14 = Anim Dynamics - Bone: dyn_R_frontCoat_3(ComponentPose=@n13, Alpha=1, ExternalForce="0,0,0") {
    angularDampingOverride=1
    angularSpringConstant=50
    bOverrideAngularDamping=true
    simulationSpace=RootRelative
    alphaBoolBlend {blendOption=Linear}
    boundBone {boneName=dyn_R_frontCoat_3}
    physicsBodyDefinitions [{
      boundBone {boneName=dyn_R_frontCoat_3}
      boxExtents=<10 10 10>
      constraintSetup {angularLimitsMax=<45 2 45>, angularLimitsMin=<-45 -40 -45>, angularTarget=<1 0 0>}
      localJointOffset=<-1 0 0>
     }]
    relativeSpaceBone {boneName=root}
   }
   n15 = "Transform (Modify) Bone - Bone: dyn_R_frontCoat_master"(ComponentPose=@n14, Alpha=1, Translation="0,0,0", Rotation="5,0,10", Scale="1,1,1") {
    rotationMode=BMM_Additive
    rotationSpace=BCS_BoneSpace
    alphaBoolBlend {blendOption=Linear}
    boneToModify {boneName=dyn_R_frontCoat_master}
    rotation {pitch=5, roll=10}
   }
   n16 = Component To Local(ComponentPose=@n15)
   n17 = Use cached pose 'postFixCache'()
   n18 = Get customIdleIsPlayingBool() -> (customIdleIsPlayingBool:bool)
   n19 = Anim Dynamics - Bone: dyn_L_front_belt_1(
    ComponentPose=Local To Component(LocalPose=Blend Poses by bool(BlendPose_0=@n16, BlendPose_1=@n17, bActiveValue=@n18) {blendType=Linear})
    Alpha=1
    ExternalForce="0,0,0"
   ) {
    angularDampingOverride=1
    angularSpringConstant=20000
    bAngularSpring=true
    bOverrideAngularDamping=true
    bUseGravityOverride=true
    gravityScale=0.35
    simulationSpace=RootRelative
    alphaBoolBlend {blendOption=Linear}
    boundBone {boneName=dyn_L_front_belt_1}
    componentLinearAccScale=<1 1 1>
    componentLinearVelScale=<1 1 1>
    gravityOverride {z=-980}
    physicsBodyDefinitions [{
      boundBone {boneName=dyn_L_front_belt_1}
      boxExtents=<20 30 30>
      constraintSetup {angularLimitsMax=<15 15 30>, angularLimitsMin=<-15 -15 -5>, angularTarget=<1 0 0>}
      localJointOffset=<2 0 0>
     }]
   }
   n20 = Anim Dynamics - Bone: dyn_L_front_belt_2(ComponentPose=@n19, Alpha=1, ExternalForce="0,0,0") {
    angularDampingOverride=1
    angularSpringConstant=20000
    bAngularSpring=true
    bOverrideAngularDamping=true
    bUseGravityOverride=true
    gravityScale=0.35
    simulationSpace=RootRelative
    alphaBoolBlend {blendOption=Linear}
    boundBone {boneName=dyn_L_front_belt_2}
    componentLinearAccScale=<1 1 1>
    componentLinearVelScale=<1 1 1>
    gravityOverride {z=-980}
    physicsBodyDefinitions [{
      boundBone {boneName=dyn_L_front_belt_2}
      boxExtents=<20 30 30>
      constraintSetup {angularLimitsMax=<15 15 30>, angularLimitsMin=<-15 -15 -5>, angularTarget=<1 0 0>}
      localJointOffset=<2 0 0>
     }]
   }
   n21 = Anim Dynamics - Bone: dyn_R_front_belt_1(ComponentPose=@n20, Alpha=1, ExternalForce="0,0,0") {
    angularDampingOverride=1
    angularSpringConstant=20000
    bAngularSpring=true
    bOverrideAngularDamping=true
    bUseGravityOverride=true
    gravityScale=0.35
    simulationSpace=RootRelative
    alphaBoolBlend {blendOption=Linear}
    boundBone {boneName=dyn_R_front_belt_1}
    componentLinearAccScale=<1 1 1>
    componentLinearVelScale=<1 1 1>
    gravityOverride {z=-980}
    physicsBodyDefinitions [{
      boundBone {boneName=dyn_R_front_belt_1}
      boxExtents=<20 30 30>
      constraintSetup {angularLimitsMax=<15 15 30>, angularLimitsMin=<-15 -15 -5>, angularTarget=<1 0 0>}
      localJointOffset=<2 0 0>
     }]
   }
   n22 = Anim Dynamics - Bone: dyn_R_front_belt_2(ComponentPose=@n21, Alpha=1, ExternalForce="0,0,0") {
    angularDampingOverride=1
    angularSpringConstant=20000
    bAngularSpring=true
    bOverrideAngularDamping=true
    bUseGravityOverride=true
    gravityScale=0.35
    simulationSpace=RootRelative
    alphaBoolBlend {blendOption=Linear}
    boundBone {boneName=dyn_R_front_belt_2}
    componentLinearAccScale=<1 1 1>
    componentLinearVelScale=<1 1 1>
    gravityOverride {z=-980}
    physicsBodyDefinitions [{
      boundBone {boneName=dyn_R_front_belt_2}
      boxExtents=<20 30 30>
      constraintSetup {angularLimitsMax=<15 15 30>, angularLimitsMin=<-15 -15 -5>, angularTarget=<1 0 0>}
      localJointOffset=<2 0 0>
     }]
   }
   n23 = Anim Dynamics - Bone: dyn_L_flap(ComponentPose=@n22, Alpha=1, ExternalForce="0,0,0") {
    angularDampingOverride=1
    angularSpringConstant=20000
    bAngularSpring=true
    bOverrideAngularDamping=true
    bUseGravityOverride=true
    gravityScale=0.35
    simulationSpace=RootRelative
    alphaBoolBlend {blendOption=Linear}
    boundBone {boneName=dyn_L_flap}
    componentLinearAccScale=<1 1 1>
    componentLinearVelScale=<1 1 1>
    gravityOverride {z=-980}
    physicsBodyDefinitions [{
      boundBone {boneName=dyn_L_flap}
      boxExtents=<20 30 30>
      constraintSetup {angularLimitsMax=<0 0 20>, angularLimitsMin=<0 0 -5>, angularTarget=<1 0 0>}
      localJointOffset=<4 0 0>
     }]
   }
   n24 = Anim Dynamics - Bone: dyn_R_flap(ComponentPose=@n23, Alpha=1, ExternalForce="0,0,0") {
    angularDampingOverride=1
    angularSpringConstant=20000
    bAngularSpring=true
    bOverrideAngularDamping=true
    bUseGravityOverride=true
    gravityScale=0.35
    simulationSpace=RootRelative
    alphaBoolBlend {blendOption=Linear}
    boundBone {boneName=dyn_R_flap}
    componentLinearAccScale=<1 1 1>
    componentLinearVelScale=<1 1 1>
    gravityOverride {z=-980}
    physicsBodyDefinitions [{
      boundBone {boneName=dyn_R_flap}
      boxExtents=<20 30 30>
      constraintSetup {angularLimitsMax=<0 0 5>, angularLimitsMin=<0 0 -20>, angularTarget=<1 0 0>}
      localJointOffset=<4 0 0>
     }]
   }
   n25 = Anim Dynamics - Bone: dyn_L_braceletMaster(ComponentPose=@n24, Alpha=1, ExternalForce="0,0,0") {
    angularDampingOverride=1
    bLinearSpring=true
    bOverrideAngularDamping=true
    bUseGravityOverride=true
    gravityScale=0.35
    linearSpringConstant=100
    simulationSpace=RootRelative
    alphaBoolBlend {blendOption=Linear}
    boundBone {boneName=dyn_L_braceletMaster}
    componentLinearAccScale=<1 1 1>
    componentLinearVelScale=<1 1 1>
    gravityOverride {z=-980}
    physicsBodyDefinitions [{
      boundBone {boneName=dyn_L_braceletMaster}
      boxExtents=<60 60 60>
      constraintSetup {angularLimitsMax=<0 0 0>, angularLimitsMin=<0 0 0>, angularTarget=<1 0 0>, linearAxesMin=<-3 0 0>}
      localJointOffset=<1 0 0>
     }]
    relativeSpaceBone {boneName=hand_l}
   }
   n26 = Anim Dynamics - Bone: dyn_L_braceletSecondary(ComponentPose=@n25, Alpha=1, ExternalForce="0,0,0") {
    angularDampingOverride=1
    angularSpringConstant=8000
    bAngularSpring=true
    bOverrideAngularDamping=true
    bUseGravityOverride=true
    gravityScale=0.35
    simulationSpace=RootRelative
    alphaBoolBlend {blendOption=Linear}
    boundBone {boneName=dyn_L_braceletSecondary}
    componentLinearAccScale=<1 1 1>
    componentLinearVelScale=<1 1 1>
    gravityOverride {z=-980}
    physicsBodyDefinitions [{
      boundBone {boneName=dyn_L_braceletSecondary}
      boxExtents=<30 30 30>
      constraintSetup {angularLimitsMax=<0 10 0>, angularLimitsMin=<0 -40 0>, angularTarget=<1 0 0>}
      localJointOffset=<1 0 0>
     }]
    relativeSpaceBone {boneName=hand_l}
   }
   n27 = Anim Dynamics - Bone: dyn_R_braceletMaster(ComponentPose=@n26, Alpha=1, ExternalForce="0,0,0") {
    angularDampingOverride=1
    bLinearSpring=true
    bOverrideAngularDamping=true
    bUseGravityOverride=true
    gravityScale=0.35
    linearSpringConstant=100
    simulationSpace=RootRelative
    alphaBoolBlend {blendOption=Linear}
    boundBone {boneName=dyn_R_braceletMaster}
    componentLinearAccScale=<1 1 1>
    componentLinearVelScale=<1 1 1>
    gravityOverride {z=-980}
    physicsBodyDefinitions [{
      boundBone {boneName=dyn_R_braceletMaster}
      boxExtents=<60 60 60>
      constraintSetup {angularLimitsMax=<0 0 0>, angularLimitsMin=<0 0 0>, angularTarget=<1 0 0>, linearAxesMin=<-3 0 0>}
      localJointOffset=<1 0 0>
     }]
    relativeSpaceBone {boneName=hand_r}
   }
   n28 = Anim Dynamics - Bone: dyn_R_braceletSecondary(ComponentPose=@n27, Alpha=1, ExternalForce="0,0,0") {
    angularDampingOverride=1
    angularSpringConstant=8000
    bAngularSpring=true
    bOverrideAngularDamping=true
    bUseGravityOverride=true
    gravityScale=0.35
    simulationSpace=RootRelative
    alphaBoolBlend {blendOption=Linear}
    boundBone {boneName=dyn_R_braceletSecondary}
    componentLinearAccScale=<1 1 1>
    componentLinearVelScale=<1 1 1>
    gravityOverride {z=-980}
    physicsBodyDefinitions [{
      boundBone {boneName=dyn_R_braceletSecondary}
      boxExtents=<30 30 30>
      constraintSetup {angularLimitsMax=<0 10 0>, angularLimitsMin=<0 -40 0>, angularTarget=<1 0 0>}
      localJointOffset=<1 0 0>
     }]
    relativeSpaceBone {boneName=hand_r}
   }
   n29 = Anim Dynamics - Bone: dyn_L_bootStrap(ComponentPose=@n28, Alpha=1, ExternalForce="0,0,0") {
    angularDampingOverride=0.999
    angularSpringConstant=500
    bAngularSpring=true
    bOverrideAngularDamping=true
    bUseGravityOverride=true
    gravityScale=0.35
    simulationSpace=RootRelative
    alphaBoolBlend {blendOption=Linear}
    boundBone {boneName=dyn_L_bootStrap}
    componentLinearAccScale=<1 1 1>
    componentLinearVelScale=<1 1 1>
    gravityOverride {z=-980}
    physicsBodyDefinitions [{
      boundBone {boneName=dyn_L_bootStrap}
      boxExtents=<10 10 10>
      constraintSetup {angularLimitsMax=<0 0 20>, angularLimitsMin=<0 0 -20>, angularTarget=<1 0 0>}
      localJointOffset=<1 0 0>
     }]
   }
   n30 = Anim Dynamics - Bone: dyn_R_bootStrap(ComponentPose=@n29, Alpha=1, ExternalForce="0,0,0") {
    angularDampingOverride=0.999
    angularSpringConstant=500
    bAngularSpring=true
    bOverrideAngularDamping=true
    bUseGravityOverride=true
    gravityScale=0.35
    simulationSpace=RootRelative
    alphaBoolBlend {blendOption=Linear}
    boundBone {boneName=dyn_R_bootStrap}
    componentLinearAccScale=<1 1 1>
    componentLinearVelScale=<1 1 1>
    gravityOverride {z=-980}
    physicsBodyDefinitions [{
      boundBone {boneName=dyn_R_bootStrap}
      boxExtents=<10 10 10>
      constraintSetup {angularLimitsMax=<0 0 20>, angularLimitsMin=<0 0 -20>, angularTarget=<1 0 0>}
      localJointOffset=<1 0 0>
     }]
   }
   n31 = Anim Dynamics - Bone: dyn_L_innerCollar_A(ComponentPose=@n30, Alpha=1, ExternalForce="0,0,0") {
    angularDampingOverride=1
    bLinearSpring=true
    bOverrideAngularDamping=true
    bOverrideLinearDamping=true
    bUseGravityOverride=true
    gravityScale=0.35
    linearDampingOverride=1
    linearSpringConstant=200
    simulationSpace=BoneRelative
    alphaBoolBlend {blendOption=Linear}
    alphaScaleBiasClamp {scale=0.4}
    boundBone {boneName=dyn_L_innerCollar_A}
    componentLinearAccScale=<0.7 0.7 0.7>
    componentLinearVelScale=<0.7 0.7 0.7>
    gravityOverride {z=-980}
    physicsBodyDefinitions [{
      boundBone {boneName=dyn_L_innerCollar_A}
      boxExtents=<7 7 7>
      constraintSetup {angularLimitsMax=<0 0 0>, angularLimitsMin=<0 0 0>, angularTarget=<1 0 0>, linearAxesMax=<0 0 2>}
      localJointOffset=<0 0 1>
     }]
    relativeSpaceBone {boneName=pelvis}
   }
   n32 = Anim Dynamics - Bone: dyn_L_outerCollar(ComponentPose=@n31, Alpha=1, ExternalForce="0,0,0") {
    angularDampingOverride=1
    angularSpringConstant=20000
    bAngularSpring=true
    bOverrideAngularDamping=true
    bUseGravityOverride=true
    gravityScale=0.35
    simulationSpace=RootRelative
    alphaBoolBlend {blendOption=Linear}
    boundBone {boneName=dyn_L_outerCollar}
    componentLinearAccScale=<1 1 1>
    componentLinearVelScale=<1 1 1>
    gravityOverride {z=-980}
    physicsBodyDefinitions [{
      boundBone {boneName=dyn_L_outerCollar}
      boxExtents=<30 30 30>
      constraintSetup {angularLimitsMax=<5 30 10>, angularLimitsMin=<-5 -30 -10>, angularTarget=<1 0 0>}
      localJointOffset=<3 0 0>
     }]
   }
   n33 = Anim Dynamics - Bone: dyn_R_innerCollar_A(ComponentPose=@n32, Alpha=1, ExternalForce="0,0,0") {
    angularDampingOverride=1
    bLinearSpring=true
    bOverrideAngularDamping=true
    bOverrideLinearDamping=true
    bUseGravityOverride=true
    gravityScale=0.35
    linearDampingOverride=1
    linearSpringConstant=200
    simulationSpace=BoneRelative
    alphaBoolBlend {blendOption=Linear}
    alphaScaleBiasClamp {scale=0.4}
    boundBone {boneName=dyn_R_innerCollar_A}
    componentLinearAccScale=<0.7 0.7 0.7>
    componentLinearVelScale=<0.7 0.7 0.7>
    gravityOverride {z=-980}
    physicsBodyDefinitions [{
      boundBone {boneName=dyn_R_innerCollar_A}
      boxExtents=<7 7 7>
      constraintSetup {angularLimitsMax=<0 0 0>, angularLimitsMin=<0 0 0>, angularTarget=<1 0 0>, linearAxesMin=<0 0 -2>}
      localJointOffset=<0 0 -1>
     }]
    relativeSpaceBone {boneName=pelvis}
   }
   n34 = Anim Dynamics - Bone: dyn_R_outerCollar(ComponentPose=@n33, Alpha=1, ExternalForce="0,0,0") {
    angularDampingOverride=1
    angularSpringConstant=20000
    bAngularSpring=true
    bOverrideAngularDamping=true
    bUseGravityOverride=true
    gravityScale=0.35
    simulationSpace=RootRelative
    alphaBoolBlend {blendOption=Linear}
    boundBone {boneName=dyn_R_outerCollar}
    componentLinearAccScale=<1 1 1>
    componentLinearVelScale=<1 1 1>
    gravityOverride {z=-980}
    physicsBodyDefinitions [{
      boundBone {boneName=dyn_R_outerCollar}
      boxExtents=<30 30 30>
      constraintSetup {angularLimitsMax=<5 30 10>, angularLimitsMin=<-5 -30 -10>, angularTarget=<1 0 0>}
      localJointOffset=<3 0 0>
     }]
   }
   Output Pose(Result=Component To Local(ComponentPose=@n34)) {layerGroup=None}
   n35 = Control Rig(Source=Copy Pose From Mesh() {bUseAttachedParent=true}) {
    alphaBoolBlend {blendOption=Linear}
    controlRigAssetReference {blueprintRigClass=&/M_MED_CrispRover_ControlRig.M_MED_CrispRover_ControlRig_C}
   }
   n36 = Get neck01 back() -> (neck01 back:double)
   n37 = "Transform (Modify) Bone - Bone: dyn_neck_01_constraint_inner"(ComponentPose=Local To Component(LocalPose=@n35), Alpha=@n36, Translation="0,0,0", Rotation="0,-35,0", Scale="1,1,1") {
    rotationMode=BMM_Additive
    rotationSpace=BCS_BoneSpace
    alphaBoolBlend {blendOption=Linear}
    boneToModify {boneName=dyn_neck_01_constraint_inner}
    rotation {yaw=-35}
   }
   n38 = Get neck02 back() -> (neck02 back:double)
   n39 = "Transform (Modify) Bone - Bone: dyn_neck_02_constraint_inner"(ComponentPose=@n37, Alpha=@n38, Translation="0,0,0", Rotation="0.000,-15.000,-0.000", Scale="1,1,1") {
    rotationMode=BMM_Additive
    rotationSpace=BCS_BoneSpace
    alphaBoolBlend {blendOption=Linear}
    boneToModify {boneName=dyn_neck_02_constraint_inner}
    rotation=<rot 0 -15.0001 0>
   }
   n40 = "Transform (Modify) Bone - Bone: dyn_neck_01_constraint_outer"(ComponentPose=@n39, Alpha=@n36, Translation="0,0,0", Rotation="0,-25,0", Scale="1,1,1") {
    rotationMode=BMM_Additive
    rotationSpace=BCS_BoneSpace
    alphaBoolBlend {blendOption=Linear}
    boneToModify {boneName=dyn_neck_01_constraint_outer}
    rotation {yaw=-25}
   }
   n41 = "Transform (Modify) Bone - Bone: dyn_neck_02_constraint_outer"(ComponentPose=@n40, Alpha=@n38, Translation="0,0,0", Rotation="0,-10,0", Scale="1,1,1") {
    rotationMode=BMM_Additive
    rotationSpace=BCS_BoneSpace
    alphaBoolBlend {blendOption=Linear}
    boneToModify {boneName=dyn_neck_02_constraint_outer}
    rotation {yaw=-10}
   }
   n42 = "Transform (Modify) Bone - Bone: dyn_L_outerCollar"(ComponentPose=@n41, Alpha=Get L clav up() -> (L clav up:double), Translation="0,0,0", Rotation="40,-10,-10", Scale="1,1,1") {
    rotationMode=BMM_Additive
    rotationSpace=BCS_BoneSpace
    alphaBoolBlend {blendOption=Linear}
    boneToModify {boneName=dyn_L_outerCollar}
    rotation=<rot 40 -10 -10>
   }
   n43 = "Transform (Modify) Bone - Bone: dyn_R_outerCollar"(ComponentPose=@n42, Alpha=Get R clav up() -> (R clav up:double), Translation="0,0,0", Rotation="-40,10,-10", Scale="1,1,1") {
    rotationMode=BMM_Additive
    rotationSpace=BCS_BoneSpace
    alphaBoolBlend {blendOption=Linear}
    boneToModify {boneName=dyn_R_outerCollar}
    rotation=<rot -40 10 -10>
   }
   n44 = "Constraint - dyn_L_frontCoat_master (P)"(ComponentPose=@n43, Alpha=1, ConstraintWeights_0=1) {
    alphaBoolBlend {blendOption=Linear}
    boneToModify {boneName=dyn_L_frontCoat_master}
    constraintSetup [{targetBone {boneName=thigh_l}}]
    constraintWeights [1]
   }
   n45 = "Constraint - dyn_L_frontCoat_1 (P)"(ComponentPose=@n44, Alpha=1, ConstraintWeights_0=1) {
    alphaBoolBlend {blendOption=Linear}
    boneToModify {boneName=dyn_L_frontCoat_1}
    constraintSetup [{targetBone {boneName=thigh_l}}]
    constraintWeights [1]
   }
   n46 = "Constraint - dyn_R_frontCoat_1 (P)"(ComponentPose=@n45, Alpha=1, ConstraintWeights_0=1) {
    alphaBoolBlend {blendOption=Linear}
    boneToModify {boneName=dyn_R_frontCoat_1}
    constraintSetup [{targetBone {boneName=thigh_r}}]
    constraintWeights [1]
   }
   n47 = "Constraint - dyn_R_frontCoat_master (P)"(ComponentPose=@n46, Alpha=1, ConstraintWeights_0=1) {
    alphaBoolBlend {blendOption=Linear}
    boneToModify {boneName=dyn_R_frontCoat_master}
    constraintSetup [{targetBone {boneName=thigh_r}}]
    constraintWeights [1]
   }
   n48 = "Constraint - dyn_L_sideCoat_master (P)"(ComponentPose=@n47, Alpha=1, ConstraintWeights_0=1) {
    alphaBoolBlend {blendOption=Linear}
    alphaScaleBiasClamp {scale=0.5}
    boneToModify {boneName=dyn_L_sideCoat_master}
    constraintSetup [{targetBone {boneName=thigh_l}}]
    constraintWeights [1]
   }
   n49 = "Constraint - dyn_R_sideCoat_master (P)"(ComponentPose=@n48, Alpha=1, ConstraintWeights_0=1) {
    alphaBoolBlend {blendOption=Linear}
    alphaScaleBiasClamp {scale=0.5}
    boneToModify {boneName=dyn_R_sideCoat_master}
    constraintSetup [{targetBone {boneName=thigh_r}}]
    constraintWeights [1]
   }
   n50 = "Constraint - dyn_L_sideCoat_1 (P)"(ComponentPose=@n49, Alpha=1, ConstraintWeights_0=1) {
    alphaBoolBlend {blendOption=Linear}
    alphaScaleBiasClamp {scale=0.5}
    boneToModify {boneName=dyn_L_sideCoat_1}
    constraintSetup [{targetBone {boneName=thigh_l}}]
    constraintWeights [1]
   }
   n51 = "Constraint - dyn_R_sideCoat_1 (P)"(ComponentPose=@n50, Alpha=1, ConstraintWeights_0=1) {
    alphaBoolBlend {blendOption=Linear}
    alphaScaleBiasClamp {scale=0.5}
    boneToModify {boneName=dyn_R_sideCoat_1}
    constraintSetup [{targetBone {boneName=thigh_r}}]
    constraintWeights [1]
   }
   n52 = "Transform (Modify) Bone - Bone: dyn_L_frontCoat_1"(ComponentPose=@n51, Alpha=Get L leg up() -> (L leg up:double), Translation="1,-2,0", Rotation="0,-15,-15", Scale="1,1,1") {
    rotationMode=BMM_Additive
    translationMode=BMM_Additive
    translationSpace=BCS_BoneSpace
    alphaBoolBlend {blendOption=Linear}
    boneToModify {boneName=dyn_L_frontCoat_1}
    rotation {roll=-15, yaw=-15}
    translation=<1 -2>
   }
   n53 = "Transform (Modify) Bone - Bone: dyn_R_frontCoat_1"(ComponentPose=@n52, Alpha=Get R leg up() -> (R leg up:double), Translation="-1,2,0", Rotation="0,15,-15", Scale="1,1,1") {
    rotationMode=BMM_Additive
    translationMode=BMM_Additive
    translationSpace=BCS_BoneSpace
    alphaBoolBlend {blendOption=Linear}
    boneToModify {boneName=dyn_R_frontCoat_1}
    rotation {roll=-15, yaw=15}
    translation=<-1 2>
   }
   n54 = Get isCrouchingRBANon() -> (isCrouchingRBANon:double)
   n55 = "Transform (Modify) Bone - Bone: dyn_R_sideCoat_master"!(ComponentPose=@n53, Alpha=@n54, Rotation="0,30,0") -> (Pose) {
    rotationMode=BMM_Additive
    rotationSpace=BCS_BoneSpace
    alphaBoolBlend {blendOption=Linear}
    boneToModify {boneName=dyn_R_sideCoat_master}
    rotation {yaw=30}
   }
   n56 = "Transform (Modify) Bone - Bone: dyn_R_sideCoat_2"!(ComponentPose=@n55, Alpha=@n54, Rotation="0,30,0") -> (Pose) {
    rotationMode=BMM_Additive
    rotationSpace=BCS_BoneSpace
    alphaBoolBlend {blendOption=Linear}
    boneToModify {boneName=dyn_R_sideCoat_2}
    rotation {yaw=30}
   }
   n57 = "Transform (Modify) Bone - Bone: dyn_R_backCoat_2"!(ComponentPose=@n56, Alpha=@n54, Rotation="30,0,0") -> (Pose) {
    rotationMode=BMM_Additive
    rotationSpace=BCS_BoneSpace
    alphaBoolBlend {blendOption=Linear}
    boneToModify {boneName=dyn_R_backCoat_2}
    rotation {pitch=30}
   }
   n58 = RigidBody!(
    ComponentPose=@n57
    Alpha=Get CustomIdleNotPlaying() -> (CustomIdleNotPlaying:double)
    OverrideWorldGravity:Vector=Get GravityOverride()
    ExternalForce
   ) -> (Pose) {
    bClampLinearTranslationLimitToRefPose=true
    bOverrideWorldGravity=true
    overridePhysicsAsset=&/M_MED_CrispRover_Physics
    simulationSpace=BaseBoneSpace
    alphaBoolBlend {blendOption=Linear}
    baseBoneRef {boneName=root}
    componentLinearAccScale=<0.2 0.2 0.2>
    componentLinearVelScale=<0.23 0.23 0.23>
    overrideWorldGravity {z=-980}
    simSpaceSettings {maxAngularAcceleration=1, maxAngularVelocity=0.1, maxLinearAcceleration=1, worldAlpha=1}
   }
   n59 = RigidBody(ComponentPose=@n58, Alpha=Get CustomIdleIsPlaying() -> (CustomIdleIsPlaying:float), OverrideWorldGravity=Get GravityOverride()) {
    bClampLinearTranslationLimitToRefPose=true
    bOverrideWorldGravity=true
    overridePhysicsAsset=&/M_MED_CrispRover_Physics_Frontend
    simulationSpace=BaseBoneSpace
    alphaBoolBlend {blendOption=Linear}
    baseBoneRef {boneName=root}
    componentLinearAccScale=<0.2 0.2 0.2>
    componentLinearVelScale=<0.23 0.23 0.23>
    overrideWorldGravity {z=-980}
    simSpaceSettings {maxAngularAcceleration=1, maxAngularVelocity=0.1, maxLinearAcceleration=1, worldAlpha=1}
   }
   n60 = Get bIsSliding()
   n61 = "Transform (Modify) Bone - Bone: dyn_L_backCoat_master"(ComponentPose=@n59, bAlphaBoolEnabled=@n60, Translation="-11,0,15", Rotation="60,0,0", Scale="1,1,1") {
    alphaInputType=Bool
    rotationMode=BMM_Additive
    rotationSpace=BCS_BoneSpace
    translationMode=BMM_Additive
    translationSpace=BCS_BoneSpace
    alphaBoolBlend {blendInTime=0.1, blendOption=Linear, blendOutTime=0.2}
    boneToModify {boneName=dyn_L_backCoat_master}
    rotation {pitch=60}
    translation {x=-11, z=15}
   }
   n62 = "Transform (Modify) Bone - Bone: dyn_R_backCoat_master"(ComponentPose=@n61, bAlphaBoolEnabled=@n60, Translation="15,0,4", Rotation="60,0,0", Scale="1,1,1") {
    alphaInputType=Bool
    rotationMode=BMM_Additive
    rotationSpace=BCS_BoneSpace
    translationMode=BMM_Additive
    translationSpace=BCS_BoneSpace
    alphaBoolBlend {blendInTime=0.1, blendOption=Linear, blendOutTime=0.2}
    boneToModify {boneName=dyn_R_backCoat_master}
    rotation {pitch=60}
    translation {x=15, z=4}
   }
   n63 = "Transform (Modify) Bone - Bone: dyn_L_backCoat_2"!(ComponentPose=@n62, bAlphaBoolEnabled:bool=@n60, Translation="-2.596,4.000,6.572", Rotation="7,7,0", Scale="1,1,1") -> (Pose) {
    alphaInputType=Bool
    rotationMode=BMM_Additive
    rotationSpace=BCS_BoneSpace
    translationMode=BMM_Additive
    translationSpace=BCS_BoneSpace
    alphaBoolBlend {blendInTime=0.1, blendOption=Linear, blendOutTime=0.2}
    boneToModify {boneName=dyn_L_backCoat_2}
    rotation {pitch=7, yaw=7}
    translation=<-2.596 4 6.572>
   }
   n64 = "Transform (Modify) Bone - Bone: dyn_R_backCoat_2"!(ComponentPose=@n63, bAlphaBoolEnabled:bool=@n60, Translation:Vector="1,-7,0", Rotation="7,7,0", Scale:Vector="1,1,1") -> (Pose) {
    alphaInputType=Bool
    rotationMode=BMM_Additive
    rotationSpace=BCS_BoneSpace
    translationMode=BMM_Additive
    translationSpace=BCS_BoneSpace
    alphaBoolBlend {blendInTime=0.1, blendOption=Linear, blendOutTime=0.2}
    boneToModify {boneName=dyn_R_backCoat_2}
    rotation {pitch=7, yaw=7}
    translation=<1 -7>
   }
   Save cached pose 'preFixCache'(Pose=Component To Local(ComponentPose=@n64)) {cacheName=preFixCache}
   n65 = "Constraint - dyn_L_frontCoat_master (P)"(ComponentPose=Local To Component(LocalPose=Use cached pose 'preFixCache'()), Alpha=1, ConstraintWeights_0=1) {
    alphaBoolBlend {blendOption=Linear}
    boneToModify {boneName=dyn_L_frontCoat_master}
    constraintSetup [{targetBone {boneName=thigh_l}}]
    constraintWeights [1]
   }
   n66 = Get L leg up()
   n67 = "Constraint - dyn_L_frontCoat_1 (P)"(ComponentPose=@n65, Alpha=@n66, ConstraintWeights_0=1) {
    alphaBoolBlend {blendOption=Linear}
    boneToModify {boneName=dyn_L_frontCoat_1}
    constraintSetup [{targetBone {boneName=thigh_l}}]
    constraintWeights [1]
   }
   n68 = "Constraint - dyn_L_frontCoat_2 (P)"(ComponentPose=@n67, Alpha=@n66, ConstraintWeights_0=1) {
    alphaBoolBlend {blendOption=Linear}
    boneToModify {boneName=dyn_L_frontCoat_2}
    constraintSetup [{targetBone {boneName=thigh_l}}]
    constraintWeights [1]
   }
   n69 = "Constraint - dyn_L_frontCoat_3 (P)"(ComponentPose=@n68, Alpha=@n66, ConstraintWeights_0=1) {
    alphaBoolBlend {blendOption=Linear}
    alphaScaleBiasClamp {scale=0.8}
    boneToModify {boneName=dyn_L_frontCoat_3}
    constraintSetup [{targetBone {boneName=thigh_l}}]
    constraintWeights [1]
   }
   n70 = "Constraint - dyn_R_frontCoat_master (P)"(ComponentPose=@n69, Alpha=1, ConstraintWeights_0=1) {
    alphaBoolBlend {blendOption=Linear}
    boneToModify {boneName=dyn_R_frontCoat_master}
    constraintSetup [{targetBone {boneName=thigh_r}}]
    constraintWeights [1]
   }
   n71 = Get R leg up()
   n72 = "Constraint - dyn_R_frontCoat_1 (P)"(ComponentPose=@n70, Alpha=@n71, ConstraintWeights_0=1) {
    alphaBoolBlend {blendOption=Linear}
    boneToModify {boneName=dyn_R_frontCoat_1}
    constraintSetup [{targetBone {boneName=thigh_r}}]
    constraintWeights [1]
   }
   n73 = "Constraint - dyn_R_frontCoat_2 (P)"(ComponentPose=@n72, Alpha=@n71, ConstraintWeights_0=1) {
    alphaBoolBlend {blendOption=Linear}
    boneToModify {boneName=dyn_R_frontCoat_2}
    constraintSetup [{targetBone {boneName=thigh_r}}]
    constraintWeights [1]
   }
   n74 = "Constraint - dyn_R_frontCoat_3 (P)"(ComponentPose=@n73, Alpha=@n71, ConstraintWeights_0=1) {
    alphaBoolBlend {blendOption=Linear}
    alphaScaleBiasClamp {scale=0.8}
    boneToModify {boneName=dyn_R_frontCoat_3}
    constraintSetup [{targetBone {boneName=thigh_r}}]
    constraintWeights [1]
   }
   n75 = "Transform (Modify) Bone - Bone: dyn_R_frontCoat_2"!(ComponentPose=@n74, bAlphaBoolEnabled:bool=Get bIsJumping(), Translation="0,0,0", Rotation="-5,0,-5", Scale="1,1,1") -> (Pose) {
    alphaInputType=Bool
    rotationMode=BMM_Additive
    alphaBoolBlend {blendInTime=0.1, blendOption=Linear, blendOutTime=0.1}
    boneToModify {boneName=dyn_R_frontCoat_2}
    rotation {pitch=-5, roll=-5}
   }
   n76 = "Transform (Modify) Bone - Bone: dyn_L_frontCoat_1"(ComponentPose=@n75, Alpha=Get L leg pitch() -> (L leg pitch:double), Translation="0,-1,6", Rotation="0,15,5", Scale="1,1,1") {
    rotationMode=BMM_Additive
    translationMode=BMM_Additive
    translationSpace=BCS_BoneSpace
    alphaBoolBlend {blendOption=Linear}
    boneToModify {boneName=dyn_L_frontCoat_1}
    rotation {roll=5, yaw=15}
    translation {y=-1, z=6}
   }
   n77 = Get L leg up()
   n78 = "Transform (Modify) Bone - Bone: dyn_L_frontCoat_1"(ComponentPose=@n76, Alpha=@n77, Translation="0,-4.8,-2.5", Rotation="-0.000,-5.000,0.000", Scale="1,1,1") {
    rotationMode=BMM_Additive
    translationMode=BMM_Additive
    translationSpace=BCS_BoneSpace
    alphaBoolBlend {blendOption=Linear}
    boneToModify {boneName=dyn_L_frontCoat_1}
    rotation {yaw=-5}
    translation {y=-4.8, z=-2.5}
   }
   n79 = "Transform (Modify) Bone - Bone: dyn_L_frontCoat_2"(ComponentPose=@n78, Alpha=@n77, Translation="0,-3,3", Rotation="0,-10,5", Scale="1,1,1") {
    rotationMode=BMM_Additive
    translationMode=BMM_Additive
    translationSpace=BCS_BoneSpace
    alphaBoolBlend {blendOption=Linear}
    boneToModify {boneName=dyn_L_frontCoat_2}
    rotation {roll=5, yaw=-10}
    translation {y=-3, z=3}
   }
   n80 = "Transform (Modify) Bone - Bone: dyn_L_sideCoat_master"(ComponentPose=@n79, Alpha=Get L leg pitch(), Translation="0,0,0", Rotation="25,0,0", Scale="1,1,1") {
    rotationMode=BMM_Additive
    alphaBoolBlend {blendOption=Linear}
    alphaScaleBiasClamp {interpSpeedDecreasing=5, interpSpeedIncreasing=50}
    boneToModify {boneName=dyn_L_sideCoat_master}
    rotation {pitch=25}
   }
   n81 = "Transform (Modify) Bone - Bone: dyn_R_frontCoat_1"(ComponentPose=@n80, Alpha=Get R leg pitch() -> (R leg pitch:double), Translation="0,1,-6", Rotation="0,-15,5", Scale="1,1,1") {
    rotationMode=BMM_Additive
    translationMode=BMM_Additive
    translationSpace=BCS_BoneSpace
    alphaBoolBlend {blendOption=Linear}
    boneToModify {boneName=dyn_R_frontCoat_1}
    rotation {roll=5, yaw=-15}
    translation {y=1, z=-6}
   }
   n82 = Get R leg up()
   n83 = "Transform (Modify) Bone - Bone: dyn_R_frontCoat_1"(ComponentPose=@n81, Alpha=@n82, Translation="0,4.8,2.5", Rotation="0,5,0", Scale="1,1,1") {
    rotationMode=BMM_Additive
    translationMode=BMM_Additive
    translationSpace=BCS_BoneSpace
    alphaBoolBlend {blendOption=Linear}
    boneToModify {boneName=dyn_R_frontCoat_1}
    rotation {yaw=5}
    translation {y=4.8, z=2.5}
   }
   n84 = "Transform (Modify) Bone - Bone: dyn_R_frontCoat_2"!(ComponentPose=@n83, Alpha:float=@n82, Translation="0,3,-3", Rotation="0.000,10.000,10.000", Scale="1,1,1") -> (Pose) {
    rotationMode=BMM_Additive
    translationMode=BMM_Additive
    translationSpace=BCS_BoneSpace
    alphaBoolBlend {blendOption=Linear}
    boneToModify {boneName=dyn_R_frontCoat_2}
    rotation=<rot 0 10 10>
    translation {y=3, z=-3}
   }
   n85 = "Transform (Modify) Bone - Bone: dyn_R_sideCoat_master"!(ComponentPose=@n84, Alpha=Get R leg pitch(), Translation:Vector="0,0,0", Rotation="-25,0,0", Scale:Vector="1,1,1") -> (Pose) {
    rotationMode=BMM_Additive
    alphaBoolBlend {blendOption=Linear}
    alphaScaleBiasClamp {interpSpeedDecreasing=5, interpSpeedIncreasing=50}
    boneToModify {boneName=dyn_R_sideCoat_master}
    rotation {pitch=-25}
   }
   n86 = Get isCrouchNotMoving() -> (isCrouchNotMoving:double)
   n87 = "Transform (Modify) Bone - Bone: dyn_R_sideCoat_master"!(ComponentPose=@n85, Alpha=@n86, Rotation="0,15,0") -> (Pose) {
    rotationMode=BMM_Additive
    rotationSpace=BCS_BoneSpace
    alphaBoolBlend {blendOption=Linear}
    alphaScaleBiasClamp {bInterpResult=true, interpSpeedDecreasing=20, interpSpeedIncreasing=20}
    boneToModify {boneName=dyn_R_sideCoat_master}
    rotation {yaw=15}
   }
   n88 = "Transform (Modify) Bone - Bone: dyn_R_sideCoat_2"(ComponentPose=@n87, Alpha=@n86, Rotation="0,15,0") {
    rotationMode=BMM_Additive
    rotationSpace=BCS_BoneSpace
    alphaBoolBlend {blendOption=Linear}
    alphaScaleBiasClamp {bInterpResult=true, interpSpeedDecreasing=20, interpSpeedIncreasing=20}
    boneToModify {boneName=dyn_R_sideCoat_2}
    rotation {yaw=15}
   }
   n89 = "Transform (Modify) Bone - Bone: dyn_R_backCoat_2"!(ComponentPose=@n88, Alpha:float=@n86, Rotation="20,0,0") -> (Pose) {
    rotationMode=BMM_Additive
    rotationSpace=BCS_BoneSpace
    alphaBoolBlend {blendOption=Linear}
    alphaScaleBiasClamp {bInterpResult=true, interpSpeedDecreasing=20, interpSpeedIncreasing=20}
    boneToModify {boneName=dyn_R_backCoat_2}
    rotation {pitch=20}
   }
   n90 = "Transform (Modify) Bone - Bone: dyn_L_backCoat_master"!(ComponentPose=@n89, Alpha:float=Get L legBack collide() -> (L legBack collide:double), Translation="0,0,0", Rotation="0,5,10", Scale="1,1,1") -> (Pose) {
    rotationMode=BMM_Additive
    alphaBoolBlend {blendOption=Linear}
    alphaScaleBiasClamp {bInterpResult=true, interpSpeedDecreasing=40, interpSpeedIncreasing=40}
    boneToModify {boneName=dyn_L_backCoat_master}
    rotation {roll=10, yaw=5}
   }
   n91 = "Transform (Modify) Bone - Bone: dyn_R_backCoat_master"!(
    ComponentPose=@n90
    Alpha:float=Get R legBack collide() -> (R legBack collide:double)
    Translation="0,0,0"
    Rotation="0.000,-5.000,10.000"
    Scale="1,1,1"
   ) -> (Pose) {
    rotationMode=BMM_Additive
    alphaBoolBlend {blendOption=Linear}
    alphaScaleBiasClamp {bInterpResult=true, interpSpeedDecreasing=40, interpSpeedIncreasing=40}
    boneToModify {boneName=dyn_R_backCoat_master}
    rotation {roll=10, yaw=-5}
   }
   Save cached pose 'postFixCache'(Pose=Component To Local(ComponentPose=@n91)) {cacheName=postFixCache}
  ]}
 EventGraph {nodes [
   Event BlueprintUpdateAnimation()
   > n1 = Sequence!(execute) -> (then_0, then_1, then_2:Exec, then_3:Exec)
   n2 = GetOwnerSkeletalMeshComponent()
   n3 = BreakRotator(InRot=BreakTransform(InTransform=GetDeltaTransformFromRefPose(self=@n2, BoneName=thigh_l, BaseName=pelvis)).Rotation)
   n4 = Get bIsCrouching()
   Set L leg up(
    execute=@n1.then_3
    L leg up:double=float * float(A=MapRangeClamped(Value=@n3.Yaw, InRangeA=0, InRangeB=-60), B=SelectFloat(A=1, B=0.35, bPickA=@n4)) {is=title K2Node_PromotableOperator}
   ) -> (Output_Get:double)
   > Set L leg pitch(L leg pitch:double=MapRangeClamped(Value=@n3.Pitch, InRangeA=0, InRangeB=40)) -> (Output_Get:double)
   n5 = BreakRotator(InRot=BreakTransform(InTransform=GetDeltaTransformFromRefPose(self=@n2, BoneName=thigh_r, BaseName=pelvis)).Rotation)
   n6 = Get bIsCrouching()
   > Set R leg up(R leg up:double=float * float(A=MapRangeClamped(Value=@n5.Yaw, InRangeA=0, InRangeB=-60), B=SelectFloat(A=1, B=0.35, bPickA=@n6))) -> (Output_Get:double)
   > Set R leg pitch(R leg pitch:double=MapRangeClamped(Value=@n5.Pitch, InRangeA=0, InRangeB=40)) -> (Output_Get:double)
   n7 = BreakTransform(InTransform=GetDeltaTransformFromRefPose(self=@n2, BoneName=clavicle_l, BaseName=spine_05))
   > Set L clav up(L clav up:double=MapRangeClamped(Value=BreakRotator(InRot=@n7.Rotation).Pitch, InRangeA=15, InRangeB=45)) -> (Output_Get:double)
   n8 = BreakTransform(InTransform=GetDeltaTransformFromRefPose(self=@n2, BoneName=clavicle_r, BaseName=spine_05))
   > Set R clav up(R clav up:double=MapRangeClamped(Value=BreakRotator(InRot=@n8.Rotation).Pitch, InRangeA=15, InRangeB=45)) -> (Output_Get:double)
   n9 = BreakRotator(InRot=BreakTransform(InTransform=GetDeltaTransformFromRefPose(self=@n2, BoneName=thigh_r, BaseName=pelvis)).Rotation)
   n10 = MapRangeClamped(Value=@n9.Yaw, InRangeA=-10, InRangeB=1)
   n11 = GetDeltaTransformFromRefPose(self=@n2, BoneName=calf_r, BaseName=thigh_r)
   n12 = MapRangeClamped(Value=BreakRotator(InRot=BreakTransform(InTransform=@n11).Rotation).Yaw, InRangeA=50, InRangeB=70)
   > Set R legBack collide(R legBack collide:double=float * float(A=MapRangeClamped(Value=@n9.Yaw, InRangeA=-15, InRangeB=-5), B=float + float(A=@n10, B=@n12))) -> (Output_Get:double)
   n13 = BreakRotator(InRot=BreakTransform(InTransform=GetDeltaTransformFromRefPose(self=@n2, BoneName=thigh_l, BaseName=pelvis)).Rotation)
   n14 = GetDeltaTransformFromRefPose(self=@n2, BoneName=calf_l, BaseName=thigh_l)
   n15 = MapRangeClamped(Value=BreakRotator(InRot=BreakTransform(InTransform=@n14).Rotation).Yaw, InRangeA=50, InRangeB=70)
   n16 = MapRangeClamped(Value=@n13.Yaw, InRangeA=-10, InRangeB=1)
   > Set L legBack collide(L legBack collide:double=float * float(A=MapRangeClamped(Value=@n13.Yaw, InRangeA=-15, InRangeB=-5), B=float + float(A=@n16, B=@n15))) -> (Output_Get:double)
   n17 = BreakTransform(InTransform=GetDeltaTransformFromRefPose(self=@n2, BoneName=neck_01, BaseName=spine_05))
   > Set neck01 back(neck01 back:double=MapRangeClamped(Value=BreakRotator(InRot=@n17.Rotation).Yaw, InRangeA=0, InRangeB=-40)) -> (Output_Get:double)
   n18 = BreakTransform(InTransform=GetDeltaTransformFromRefPose(self=@n2, BoneName=neck_02, BaseName=neck_01))
   > Set neck02 back(neck02 back:double=MapRangeClamped(Value=BreakRotator(InRot=@n18.Rotation).Yaw, InRangeA=0, InRangeB=-5)) -> (Output_Get:double)
   n19 = AND Boolean(A=Get bIsCrouching(), B=NOT Boolean(A=Get bIsAccelerating2D()))
   > Set isCrouchNotMoving(isCrouchNotMoving:double=float * float(A="To Float (Boolean)"(InBool=@n19), B=Get perfAlpha() -> (perfAlpha:double))) -> (Output_Get:double)
   n20 = Get bIsCrouching()
   > Set isCrouchingRBANon(isCrouchingRBANon:double=float * float(A="To Float (Boolean)"(InBool=@n20), B=Get perfAlpha())) -> (Output_Get:double)
   n21 = Branch(execute=@n1.then_0, Condition=Get bIsInFrontEnd())
   n22 = GetCurveValue(self=GetOwnerFrontEndAnimInstance(), CurveName=Frontend_Idle_Bypass)
   > n23 = Set CustomIdleIsPlaying(CustomIdleIsPlaying:float=@n22) -> (Output_Get:float)
   > Set CustomIdleNotPlaying(CustomIdleNotPlaying:double=float - float(A=1, B=@n23)) -> (Output_Get:double)
   > Set customIdleIsPlayingBool(customIdleIsPlayingBool:bool="float >= float"(A=@n22, B=1)) -> (Output_Get:bool)
   Set CustomIdleNotPlaying(execute=@n21.else, CustomIdleNotPlaying=1)
   > Set customIdleIsPlayingBool(customIdleIsPlayingBool=false)
   Set perfAlpha(execute=@n1.then_1, perfAlpha:double="To Float (Boolean)"(InBool=Get bRigidBodyEnabled())) -> (Output_Get:double)
   n24 = Gravity Override State Function DEPRECATED(
    execute=@n1.then_2
    CharacterPartAnimInstance=Self-Reference()
    GameTimeInSeconds=GetGameTimeInSeconds()
    GravityOverrideParams=Get Gravity Override Params()
   )
   > Set GravityOverride(GravityOverride=@n24)
   > Branch(Condition=Get bIsJumping())
   > Set GravityOverride(GravityOverride="0,0,100")
   n25 = Event OnEndCrouching()
   ResetDynamics(execute=[@n25.then, @n26.then, @n27.then], InTeleportType=TeleportPhysics)
   > ResetDynamics(InTeleportType=ResetPhysics)
   n26 = Event OnBeginCrouching()
   Event OnEndFalling()
   > n27 = Branch(Condition=NOT Boolean(A=Get bIsSkyDiving()))
  ]}
]
variables [
 CustomIdleIsPlaying {type=number}
 customIdleIsPlayingBool {type=boolean}
 CustomIdleNotPlaying {type=number}
 Gravity Override Params {
  type=GravityOverrideParamsStruct
  default {onGround_Falling {gravityOverride=<0 0 -100>}, riding_Moving {gravityOverride=<0 0 -600>}, tactical_Sprint {gravityOverride=<0 0 -980>}}
 }
 GravityOverride {type=Vector}
 isCrouchingRBANon {type=number}
 isCrouchNotMoving {type=number}
 L clav up {type=number}
 L leg pitch {type=number}
 L leg up {type=number}
 L legBack collide {type=number}
 neck01 back {type=number}
 neck02 back {type=number}
 perfAlpha {type=number}
 R clav up {type=number}
 R leg pitch {type=number}
 R leg up {type=number}
 R legBack collide {type=number}
]
