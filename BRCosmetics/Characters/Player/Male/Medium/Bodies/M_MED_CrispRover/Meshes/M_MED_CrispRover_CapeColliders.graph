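# Physics asset of cape colliders for the CrispRover medium male player body, a collision-only rig with no constraints.
# Capsules cover the pelvis, spine, neck and clavicles, with a sphere on the head, while arm, leg and foot bodies carry no shapes.
# Mirrored, offset capsules on the dynamic front, side and back coat bones add collision that follows the coat.
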PhysicsAsset /BRCosmetics/Characters/Player/Male/Medium/Bodies/M_MED_CrispRover/Meshes/M_MED_CrispRover_CapeColliders
bodies {
 calf_l []
 calf_r []
 clavicle_l [{shapeType=Capsule, length=9.0192, radius=10.5, center=<9.4266 -1.1806 -2.4977>, rotation=<rot 69.3297 1.4476 176.9462>}]
 clavicle_r [{shapeType=Capsule, length=9.0192, radius=10.5, center=<-9.4266 1.1806 2.4977>, rotation=<rot 69.3297 1.4476 -3.0538>}]
 dyn_L_backCoat_1 [{shapeType=Capsule, length=6, radius=11, shapeName=dyn_L_backCoat_1_capsule, center=<5.7087 2.6286 -5.8672>, rotation=<rot 0 0 90>}]
 dyn_L_backCoat_2 [{shapeType=Capsule, length=10, radius=11, shapeName=dyn_L_backCoat_2_capsule, center=<5.8813 2.6683 -5.881>, rotation=<rot 0 0 90>}]
 dyn_L_backCoat_3 [{shapeType=Capsule, length=15, radius=11, shapeName=dyn_L_backCoat_3_capsule, center=<5.8813 2.6683 -5.881>, rotation=<rot 0 0 90>}]
 dyn_L_frontCoat_2 [{shapeType=Capsule, length=10, radius=10, shapeName=dyn_L_frontCoat_2_capsule, center=<7.0602 -2.9502 -5.3063>, rotation=<rot 0 0 90>}]
 dyn_L_frontCoat_3 [{shapeType=Capsule, length=10, radius=10, shapeName=dyn_L_frontCoat_3_capsule, center=<7.0602 -2.9502 -5.3063>, rotation=<rot 0 0 90>}]
 dyn_L_sideCoat_1 [{shapeType=Capsule, length=10, radius=11, shapeName=dyn_L_sideCoat_1_capsule, center=<5 0 -10>, rotation=<rot 0 0 90>}]
 dyn_L_sideCoat_2 [{shapeType=Capsule, length=10, radius=11, shapeName=dyn_L_sideCoat_2_capsule, center=<5 0 -10>, rotation=<rot 0 0 90>}]
 dyn_L_sideCoat_3 [{shapeType=Capsule, length=10, radius=11, shapeName=dyn_L_sideCoat_3_capsule, center=<5 0 -10>, rotation=<rot 0 0 90>}]
 dyn_R_backCoat_1 [{shapeType=Capsule, length=6, radius=11, shapeName=dyn_R_backCoat_1_capsule, center=<-5.7087 -2.6286 5.8672>, rotation=<rot 0 0 90>}]
 dyn_R_backCoat_2 [{shapeType=Capsule, length=10, radius=11, shapeName=dyn_R_backCoat_2_capsule, center=<-5.8813 -2.6683 5.881>, rotation=<rot 0 0 90>}]
 dyn_R_backCoat_3 [{shapeType=Capsule, length=15, radius=11, shapeName=dyn_R_backCoat_3_capsule, center=<-5.8813 -2.6683 5.881>, rotation=<rot 0 0 90>}]
 dyn_R_frontCoat_2 [{shapeType=Capsule, length=10, radius=10, shapeName=dyn_R_frontCoat_2_capsule, center=<-7.0602 2.9502 5.3063>, rotation=<rot 0 0 90>}]
 dyn_R_frontCoat_3 [{shapeType=Capsule, length=10, radius=10, shapeName=dyn_R_frontCoat_3_capsule, center=<-7.0602 2.9502 5.3063>, rotation=<rot 0 0 90>}]
 dyn_R_sideCoat_1 [{shapeType=Capsule, length=10, radius=11, shapeName=dyn_R_sideCoat_1_capsule, center=<-5 0 10>, rotation=<rot 0 0 90>}]
 dyn_R_sideCoat_2 [{shapeType=Capsule, length=10, radius=11, shapeName=dyn_R_sideCoat_2_capsule, center=<-5 0 10>, rotation=<rot 0 0 90>}]
 dyn_R_sideCoat_3 [{shapeType=Capsule, length=10, radius=11, shapeName=dyn_R_sideCoat_3_capsule, center=<-5 0 10>, rotation=<rot 0 0 90>}]
 foot_l []
 foot_r []
 head [{shapeType=Sphere, radius=17.3125, center=<11.028 -0.174 0>}]
 lowerarm_l []
 lowerarm_r []
 neck_02 [{shapeType=Capsule, length=13.4375, radius=7.3175, center=<-0.2642 -0.3408 0>, rotation=<rot 90 -171.1193 -175.3978>}]
 pelvis [{shapeType=Capsule, length=4.1456, radius=13.5388, center=<-3.197 0.7838 0>}]
 spine_01 [{shapeType=Capsule, length=5.4438, radius=13.0586, center=<-1.7965 0.8851 0>}]
 spine_02 [{shapeType=Capsule, length=4.7826, radius=13.6322, center=<-1.711 2.4708 0.0001>}]
 spine_03 [{shapeType=Capsule, length=7.7107, radius=12.2031, center=<0.1445 1.3501 0.0001>}]
 spine_04 [{shapeType=Capsule, length=8.4404, radius=14.6325, center=<2.766 1.9588 0.0002>}]
 spine_05 [{shapeType=Capsule, length=5.5226, radius=16.2931, center=<-0.6513 3.0377 0>}]
 thigh_l []
 thigh_r []
 upperarm_l []
 upperarm_r []
}
constraints []
